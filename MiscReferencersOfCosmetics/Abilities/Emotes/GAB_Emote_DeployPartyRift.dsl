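; Emote gameplay ability that deploys a party rift in front of the player.
; CanActivateAbility runs the parent check, then asks the game state's PartyRiftPortalManager whether the player controller can spawn a portal, with a placement-failure tag as the fallback.
; On activation, only with authority, the manager determines location and rotation.
; If that succeeds, it spawns the PartyRiftRef actor there, owned by the player's controller.
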(event Ability|EventActivateAbility
  (bind _playerpawn (Variables|Default|GetPlayerPawn))
  (bind _returnvalue (Fortnite|Athena|GameState|GetGameStateAthenafromContext _playerpawn))
  (bind _returnvalue_2 (Class|FortGameStateAthena|GetPartyRiftPortalManager _returnvalue))
  (bind _playerpawn_2 (Variables|Default|GetPlayerPawn))
  (bind (_location _rotation _returnvalue_3) (Class|FortPartyRiftPortalManager|DeterminePortalSpawnLocation _returnvalue_2 _playerpawn_2))
  ;[0.1458 0.2535 1] ActivateAbility - spawn a party rift in front of the player
  (|CallParentFunction)
  (Ability|HasAuthority
    (:True
      (Utilities|IsValid _playerpawn
        (:Is%20Valid
          (Utilities|IsValid _returnvalue
            (:Is%20Valid
              (Utilities|IsValid _returnvalue_2
                (:Is%20Valid
                  (if _returnvalue_3
                    (Game|SpawnActorfromClass (Variables|Default|GetPartyRiftRef) (Math|Transform|MakeTransform _location _rotation) "Undefined" "SelectDefaultAtRuntime" (Fortnite|GetFortPlayerControllerfromActor (Variables|Default|GetPlayerPawn)) 0)))))))))))

(fn CanActivateAbility
  (params (ActorInfo :type Gameplay%20Ability%20Actor%20Info%20Structure) (Handle :type Gameplay%20Ability%20Spec%20Handle%20Structure))
  (bind (_owneractor _avataractor _playercontroller) (Utilities|Struct|BreakGameplayAbilityActorInfo ActorInfo))
  (bind (_relevanttags _returnvalue) (|CallParentFunction ActorInfo Handle))
  (if _returnvalue
    (bind _asfort_player_controller_athena (Utilities|Casting|CastToFortPlayerControllerAthena _playercontroller)
      (:then
        (Utilities|IsValid (Fortnite|Athena|GameState|GetGameStateAthenafromContext _asfort_player_controller_athena)
          (:Is%20Valid
            (Utilities|IsValid (Class|FortGameStateAthena|GetPartyRiftPortalManager (Fortnite|Athena|GameState|GetGameStateAthenafromContext _asfort_player_controller_athena))
              (:Is%20Valid
                (return (Class|FortPartyRiftPortalManager|CanSpawnPortal (Class|FortGameStateAthena|GetPartyRiftPortalManager (Fortnite|Athena|GameState|GetGameStateAthenafromContext _asfort_player_controller_athena)) _asfort_player_controller_athena) (.ReturnValue (Class|FortPartyRiftPortalManager|CanSpawnPortal (Class|FortGameStateAthena|GetPartyRiftPortalManager (Fortnite|Athena|GameState|GetGameStateAthenafromContext _asfort_player_controller_athena)) _asfort_player_controller_athena))))
              (:Is%20Not%20Valid
                _)))
          (:Is%20Not%20Valid
            _))
        (return "(GameplayTags=((TagName=\"Activate.Failed.Placement\")))" false))
      (:CastFailed
        _))
    (else
      (return _relevanttags _returnvalue))))
