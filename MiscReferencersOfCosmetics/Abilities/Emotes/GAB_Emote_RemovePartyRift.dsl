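; Emote gameplay ability that removes a party rift the player previously deployed.
; On activation it calls the parent, then proceeds only on the non-authority side with a valid player pawn.
; It gets the pawn's Athena player controller, checks for a valid party rift portal, and sends a server interact request for that portal tagged "Emote".
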(event Ability|EventActivateAbility
  (bind _playerpawn (Variables|Default|GetPlayerPawn))
  ;[0.1458 0.2535 1] ActivateAbility - removes any party rift the player has previously deployed
  (|CallParentFunction)
  (Ability|HasAuthority
    (:False
      (Utilities|IsValid _playerpawn
        (:Is%20Valid
          (bind _asfort_player_controller_athena (Utilities|Casting|CastToFortPlayerControllerAthena (Fortnite|GetFortPlayerControllerfromActor _playerpawn)))
          (if (Class|FortPlayerControllerAthena|HasPartyRiftPortal _asfort_player_controller_athena)
            (Utilities|IsValid (Class|FortPlayerControllerAthena|GetPartyRiftPortal _asfort_player_controller_athena)
              (:Is%20Valid
                (Class|FortPlayerControllerAthena|ServerStartInteractwithPartyRiftPortal _asfort_player_controller_athena (Class|FortPlayerControllerAthena|GetPartyRiftPortal _asfort_player_controller_athena) "Emote")))))))))
